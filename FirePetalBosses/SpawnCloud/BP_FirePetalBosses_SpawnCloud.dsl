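; Spawn cloud actor for Fire Petal Bosses, tied to the lifetime of its instigator.
; On BeginPlay, it binds to a valid instigator's OnDestroyed so that it destroys itself along with that instigator.
; It then checks the instigator for the FirePetalBosses.Demons.IsChallenged gameplay tag once, calling StopCloud if present and StartCloud otherwise.
; StartCloud and StopCloud are empty events here, so this graph holds no cloud behavior itself.
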(event EventBeginPlay
  (bind _returnvalue (Game|GetInstigator))
  (Utilities|IsValid _returnvalue
    (:Is%20Valid
      (Game|BindEventtoOnDestroyed _returnvalue (EventDispatchers|CreateEvent :!event Custom|InstigatorDestroyed))
      (bind _returnvalue_2 (GameplayTags|HasMatchingGameplayTag(Message) _returnvalue "(TagName=\"FirePetalBosses.Demons.IsChallenged\")"))
      (if _returnvalue_2
        (CallFunction|StopCloud)
        (else
          (CallFunction|StartCloud))))))

(event Custom|StartCloud)

(event Custom|StopCloud)

(event Custom|InstigatorDestroyed
  (params (DestroyedActor :type Actor%20Object%20Reference))
  (Actor|DestroyActor))

(fn ConstructionScript)
